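# Looping sound cue for the activated state of the CleanSweep item, classed as a gadget sound.
# An IsEnemy branch plays a volume-boosted mono loop for enemies, and the standard loop when the parameter is false or unset.
# An Enveloper above the branch fades the sound in over 1.5 seconds.
# It borrows the Cloak ambient loop attenuation and overrides concurrency to stop the quietest instance past the cap.
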SoundCue /CleanSweepDiagram/Audio/IP_CleanSweep_Activated_Loop
root=Enveloper_0
nodes [
 {id=Branch_0, children [Modulator_0,WavePlayer_1,WavePlayer_1], properties {boolParameterName=IsEnemy}}
 {
  id=Enveloper_0
  children [Branch_0]
  properties {pitchCurve {editorCurveData {keys [{time=0}]}}, volumeCurve {editorCurveData {keys [<at 0 0>,<at 1.5 1>]}}}
 }
 {id=Modulator_0, children [WavePlayer_0], properties {pitchMax=1, pitchMin=1, volumeMax=3, volumeMin=3}}
 {id=WavePlayer_0, properties {bLooping=true, soundWaveAssetPtr=&/IP_CleanSweep_Activate_Loop_Mono}}
 {id=WavePlayer_1, properties {bLooping=true, soundWaveAssetPtr=&/IP_CleanSweep_Activate_Loop}}
]
properties {
 attenuationSettings=/CloakGameplay/Audio/SFX/Fort_Cloak_Ambient_Loop_Att
 bOverrideConcurrency=true
 maxDistance=5300
 soundClassObject=&Fort_Gadgets
 volumeMultiplier=0.1
 concurrencyOverrides {maxCount=8, resolutionRule=StopQuietest}
}
