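# Sound cue for the CleanSweep cooldown-finished sound, named for signalling that the cooldown has ended.
# A single wave player, set as the root, plays the IP_CleanSweep_Cooldown_Finished wave with no mixing or randomisation.
# The cue routes through the Fort_Gadgets sound class at a reduced volume, with a very large max distance.
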SoundCue /CleanSweepDiagram/Audio/IP_CleanSweep_Cooldown_Finished_Cue
root=WavePlayer_0
nodes [{id=WavePlayer_0, properties {soundWaveAssetPtr=&/IP_CleanSweep_Cooldown_Finished}}]
properties {maxDistance=2097152, soundClassObject=&Fort_Gadgets, volumeMultiplier=0.2}
