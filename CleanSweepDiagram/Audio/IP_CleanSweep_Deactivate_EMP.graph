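# Sound cue for the CleanSweep EMP deactivation effect, classed as a gadget sound.
# A Random node picks among five equally weighted wave variations and feeds a Modulator at the root for per-play variation.
# The Random node is excluded from branch culling.
# Attenuation is borrowed from the Cloak activation settings, and overall volume is slightly reduced.
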SoundCue /CleanSweepDiagram/Audio/IP_CleanSweep_Deactivate_EMP
root=Modulator_0
nodes [
 {id=Modulator_0, children [Random_0]}
 {
  id=Random_0
  children [WavePlayer_1,WavePlayer_2,WavePlayer_3,WavePlayer_6,WavePlayer_7]
  properties {bShouldExcludeFromBranchCulling=true, weights [1,1,1,1,1]}
 }
 {id=WavePlayer_1, properties {soundWaveAssetPtr=&/IP_CleanSweep_Deactivate_EMP_05}}
 {id=WavePlayer_2, properties {soundWaveAssetPtr=&/IP_CleanSweep_Deactivate_EMP_04}}
 {id=WavePlayer_3, properties {soundWaveAssetPtr=&/IP_CleanSweep_Deactivate_EMP_03}}
 {id=WavePlayer_6, properties {soundWaveAssetPtr=&/IP_CleanSweep_Deactivate_EMP_02}}
 {id=WavePlayer_7, properties {soundWaveAssetPtr=&/IP_CleanSweep_Deactivate_EMP_01}}
]
properties {
 attenuationSettings=/CloakGameplay/Audio/SFX/Fort_Cloak_Activate_Att
 maxDistance=6500
 soundClassObject=&Fort_Gadgets
 volumeMultiplier=0.8
}
