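# Sound cue for the CleanSweep deactivate warning sound effect.
# A Random node picks among four equally weighted warning wave variations and is excluded from branch culling.
# The cue plays quietly in the Fort_Gadgets sound class and borrows the Cloak activate attenuation settings for spatial falloff.
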SoundCue /CleanSweepDiagram/Audio/IP_CleanSweep_Deactivate_Warning
root=Random_0
nodes [
 {
  id=Random_0
  children [WavePlayer_0, WavePlayer_13, WavePlayer_14, WavePlayer_15]
  properties {bShouldExcludeFromBranchCulling=true, weights [1,1,1,1]}
 }
 {id=WavePlayer_0, properties {soundWaveAssetPtr=&/IP_CleanSweep_Deactivate_Warning_04}}
 {id=WavePlayer_13, properties {soundWaveAssetPtr=&/IP_CleanSweep_Deactivate_Warning_03}}
 {id=WavePlayer_14, properties {soundWaveAssetPtr=&/IP_CleanSweep_Deactivate_Warning_02}}
 {id=WavePlayer_15, properties {soundWaveAssetPtr=&/IP_CleanSweep_Deactivate_Warning_01}}
]
properties {
 attenuationSettings=/CloakGameplay/Audio/SFX/Fort_Cloak_Activate_Att
 maxDistance=6500
 soundClassObject=&Fort_Gadgets
 volumeMultiplier=0.3
}
