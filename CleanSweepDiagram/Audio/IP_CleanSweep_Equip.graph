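# Sound cue that plays the equip foley for the CleanSweep item.
# A Random node picks one of three equip waves at equal weight and feeds a Modulator that slightly varies volume while keeping pitch fixed.
# It uses the weapon foley sound class and borrows weapon reload attenuation plus the global and local Bow concurrency rules.
# The cue's debug flag is left on.
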SoundCue /CleanSweepDiagram/Audio/IP_CleanSweep_Equip
root=Modulator_1
nodes [
 {id=Modulator_1, children [Random_0], properties {pitchMax=1, pitchMin=1, volumeMax=0.95, volumeMin=0.85}}
 {id=Random_0, children [WavePlayer_1,WavePlayer_2,WavePlayer_3], properties {bShouldExcludeFromBranchCulling=true, weights [1,1,1]}}
 {id=WavePlayer_1, properties {soundWaveAssetPtr=&/IP_CleanSweep_Equip_03}}
 {id=WavePlayer_2, properties {soundWaveAssetPtr=&/IP_CleanSweep_Equip_02}}
 {id=WavePlayer_3, properties {soundWaveAssetPtr=&/IP_CleanSweep_Equip_01}}
]
properties {
 attenuationSettings=&Fort_Weapon_Reload_Att
 bDebug=true
 maxDistance=2250
 soundClassObject=&Fort_WeaponFoley
 volumeMultiplier=1.002
 concurrencySet [/Game/Sounds/Concurrency_Rules/Bow_Global_Conc, /Game/Sounds/Concurrency_Rules/Bow_Local_Conc]
}
