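# Sound cue for picking up the CleanSweep item, routed through the loot-collect sound class with weapon-pickup attenuation.
# A Random node picks one of three equally weighted wave variations and feeds a root Modulator that keeps pitch fixed.
# The cue references the PickupMaximum_Oldest concurrency set to limit overlapping pickup sounds.
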SoundCue /CleanSweepDiagram/Audio/IP_CleanSweep_Pickup
root=Modulator_0
nodes [
 {id=Modulator_0, children [Random_1], properties {pitchMax=1, pitchMin=1}}
 {id=Random_1, children [WavePlayer_1,WavePlayer_3,WavePlayer_4], properties {bShouldExcludeFromBranchCulling=true, weights [1,1,1]}}
 {id=WavePlayer_1, properties {soundWaveAssetPtr=&/IP_CleanSweep_Pickup_01}}
 {id=WavePlayer_3, properties {soundWaveAssetPtr=&/IP_CleanSweep_Pickup_02}}
 {id=WavePlayer_4, properties {soundWaveAssetPtr=&/IP_CleanSweep_Pickup_03}}
]
properties {
 attenuationSettings=&Fort_Weapon_Pickup_Att
 maxDistance=3500
 soundClassObject=&Fort_Loot_Collect
 volumeMultiplier=1.4
 concurrencyOverrides {maxCount=1, resolutionRule=PreventNew}
 concurrencySet [&PickupMaximum_Oldest]
}
