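# Sound cue for the CleanSweep gadget's activation, playing one of four equally weighted waves after a very short fixed delay with slight pitch variation.
# A branch on IsLocal plays that chain directly when true, and through an extra volume-lowering modulator when false or unset.
# It borrows the cloak activation attenuation, uses the gadgets sound class and caps concurrency by stopping the oldest instance.
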SoundCue /CleanSweepDiagram/Audio/IP_Cleansweep_Activate
root=Branch_0
nodes [
 {id=Branch_0, children [Modulator_0,Modulator_1,Modulator_1], properties {boolParameterName=IsLocal}}
 {id=Delay_0, children [Random_2], properties {delayMax=0.05, delayMin=0.05}}
 {id=Modulator_0, children [Delay_0], properties {pitchMin=0.98}}
 {id=Modulator_1, children [Modulator_0], properties {pitchMax=1, pitchMin=1, volumeMax=0.65, volumeMin=0.6}}
 {
  id=Random_2
  children [WavePlayer_1, WavePlayer_2, WavePlayer_9, WavePlayer_10]
  properties {bShouldExcludeFromBranchCulling=true, weights [1,1,1,1]}
 }
 {id=WavePlayer_1, properties {soundWaveAssetPtr=&/IP_CleanSweep_Activate_01}}
 {id=WavePlayer_10, properties {soundWaveAssetPtr=&/IP_CleanSweep_Activate_04}}
 {id=WavePlayer_2, properties {soundWaveAssetPtr=&/IP_CleanSweep_Activate_02}}
 {id=WavePlayer_9, properties {soundWaveAssetPtr=&/IP_CleanSweep_Activate_03}}
]
properties {
 attenuationSettings=/CloakGameplay/Audio/SFX/Fort_Cloak_Activate_Att
 bOverrideConcurrency=true
 maxDistance=6500
 soundClassObject=&Fort_Gadgets
 volumeMultiplier=0.7
 concurrencyOverrides {maxCount=4, resolutionRule=StopOldest, volumeScaleAttackTime=0.33}
}
